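# Sound cue that plays laugh sounds for the AnglePatch emote, classed as emote foley.
# A Random node picks one of three wave variations, with the first weighted to play less often than the other two.
# It uses the shared emote attenuation and foley property sheet, overrides concurrency with a small instance cap, keeps playing when silent, and has a send to a music audio-analysis submix.
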SoundCue /BRCosmetics/Sounds/Emotes/AnglePatch/EmoteFoley_AnglePatch_Laugh
root=Random_0
nodes [
 {id=Random_0, children [WavePlayer_3,WavePlayer_4,WavePlayer_5], properties {weights [0.25,1,1]}}
 {id=WavePlayer_3, properties {soundWaveAssetPtr=&/IP_Emote_AnglePatch_Laugh_03}}
 {id=WavePlayer_4, properties {soundWaveAssetPtr=&/IP_Emote_AnglePatch_Laugh_04}}
 {id=WavePlayer_5, properties {soundWaveAssetPtr=&/IP_Emote_AnglePatch_Laugh_05}}
]
properties {
 attenuationSettings=&Fort_Emotes_Athena_Att
 audioPropertiesSheet=/BRCosmetics/Sounds/Settings/PropertySheets/Emotes/APS_Cosmetic_Emote_Foley
 bOverrideConcurrency=true
 maxDistance=1600
 soundClassObject=/Game/Sounds/FortSoundClassesAndModes/Fort_Foley_Emote
 virtualizationMode=PlayWhenSilent
 volumeMultiplier=1.25
 concurrencyOverrides {maxCount=3}
 soundSubmixSends [{minSendLevel=0, soundSubmix=&Fortnite_AudioAnalysis_Music}]
}
